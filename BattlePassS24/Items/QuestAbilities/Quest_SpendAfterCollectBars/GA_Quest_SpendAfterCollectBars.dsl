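; Quest gameplay ability that marks the owning player as having collected gold bars, apparently for a spend-after-collect quest.
; On event activation it calls the parent, then casts the event target to a pickup and checks it against the gold bar tag container.
; A match applies the collected-bars GameplayEffect to the owning pawn's ability system component, while a non-match ends the ability.
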(event Ability|EventActivateAbilityFromEvent
  (params (EventData_EventTag :type Gameplay%20Tag%20Structure) (EventData_Instigator :type Actor%20Object%20Reference) (EventData_Target :type Actor%20Object%20Reference) (EventData_OptionalObject :type Object%20Reference) (EventData_OptionalObject2 :type Object%20Reference) (EventData_ContextHandle :type Gameplay%20Effect%20Context%20Handle%20Structure) (EventData_InstigatorTags :type Gameplay%20Tag%20Container%20Structure) (EventData_TargetTags :type Gameplay%20Tag%20Container%20Structure) (EventData_EventMagnitude :type Float%20%28single-precision%29) (EventData_TargetData :type Gameplay%20Ability%20Target%20Data%20Handle%20Structure) (EventData_InstancedEventData :type Instanced%20Struct%20Structure))
  (|CallParentFunction (Utilities|Struct|MakeGameplayEventData EventData_EventTag EventData_Instigator EventData_Target EventData_OptionalObject EventData_OptionalObject2 EventData_ContextHandle EventData_InstigatorTags EventData_TargetTags EventData_EventMagnitude EventData_TargetData))
  (bind _asfort_pickup (Utilities|Casting|CastToFortPickup EventData_Target))
  (bind _returnvalue (GameplayTags|HasAnyMatchingGameplayTags(Message) _asfort_pickup (Variables|Default|GetGold%20BarTag%20Container)))
  (if _returnvalue
    (GameplayEffects|ApplyGameplayEffectToSelf :self (Ability|GetFortAbilitySystemComponent (Variables|GAVars|Data|GetOwningPlayerPawn)) :GameplayEffectClass (Variables|Default|GetGE_Quest_CollectedBars) :Level 0.0)
    (Ability|EndAbility)))
